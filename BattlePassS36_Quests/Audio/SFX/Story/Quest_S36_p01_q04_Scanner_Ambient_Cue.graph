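# Sound Cue for the Season 36 story quest scanner ambience, built from beacon ambience waves.
# A Mixer layers a looping bed with five one-shot layers (A–E), balanced by per-input volume.
# Each layer loops a random delay into a random wave pick, all but D passing through a Modulator, so accents recur irregularly.
# It uses the gadgets sound class and borrows supply-drop ambient loop attenuation.
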SoundCue /BattlePassS36_Quests/Audio/SFX/Story/Quest_S36_p01_q04_Scanner_Ambient_Cue
root=Mixer_0
nodes [
 {id=Delay_0, children [Modulator_1], properties {delayMax=8, delayMin=4}}
 {id=Delay_1, children [Modulator_2], properties {delayMax=6, delayMin=3}}
 {id=Delay_2, children [Random_3], properties {delayMax=10, delayMin=2}}
 {id=Delay_3, children [Modulator_4], properties {delayMax=4, delayMin=2}}
 {id=Delay_4, children [Modulator_0], properties {delayMax=14, delayMin=7}}
 {id=Looping_0, children [Delay_0]}
 {id=Looping_1, children [Delay_1]}
 {id=Looping_2, children [Delay_3]}
 {id=Looping_3, children [Delay_2]}
 {id=Looping_4, children [Delay_4]}
 {id=Mixer_0, children [WavePlayer_0,Looping_0,Looping_1,Looping_2,Looping_3,Looping_4], properties {inputVolume [1,1,1,1.25,0.25,0.5]}}
 {id=Modulator_0, children [Random_4], properties {pitchMax=1, pitchMin=1}}
 {id=Modulator_1, children [Random_0], properties {pitchMin=1}}
 {id=Modulator_2, children [Random_1], properties {pitchMax=1, pitchMin=1}}
 {id=Modulator_4, children [Random_2], properties {pitchMax=1, pitchMin=1}}
 {id=Random_0, children [WavePlayer_1,WavePlayer_2,WavePlayer_3], properties {weights [1,1,1]}}
 {id=Random_1, children [WavePlayer_4,WavePlayer_5,WavePlayer_6], properties {weights [1,1,1]}}
 {id=Random_2, children [WavePlayer_7,WavePlayer_8,WavePlayer_9], properties {weights [1,1,1]}}
 {id=Random_3, children [WavePlayer_10, WavePlayer_11, WavePlayer_12, WavePlayer_13], properties {weights [1,1,1,1]}}
 {id=Random_4, children [WavePlayer_14, WavePlayer_15], properties {weights [1,1]}}
 {
  id=WavePlayer_0
  properties {bLooping=true, soundWaveAssetPtr=/BattlePassS36_Quests/Audio/SFX/Story/Waves/Quest_S36_snapshot_p01_q04_beacon_ambience_01}
 }
 {id=WavePlayer_1, properties {soundWaveAssetPtr=/BattlePassS36_Quests/Audio/SFX/Story/Waves/Quest_S36_snapshot_p01_q04_beacon_ambience_A_01}}
 {id=WavePlayer_10, properties {soundWaveAssetPtr=/BattlePassS36_Quests/Audio/SFX/Story/Waves/Quest_S36_snapshot_p01_q04_beacon_ambience_D_01}}
 {id=WavePlayer_11, properties {soundWaveAssetPtr=/BattlePassS36_Quests/Audio/SFX/Story/Waves/Quest_S36_snapshot_p01_q04_beacon_ambience_D_02}}
 {id=WavePlayer_12, properties {soundWaveAssetPtr=/BattlePassS36_Quests/Audio/SFX/Story/Waves/Quest_S36_snapshot_p01_q04_beacon_ambience_D_03}}
 {id=WavePlayer_13, properties {soundWaveAssetPtr=/BattlePassS36_Quests/Audio/SFX/Story/Waves/Quest_S36_snapshot_p01_q04_beacon_ambience_D_04}}
 {id=WavePlayer_14, properties {soundWaveAssetPtr=/BattlePassS36_Quests/Audio/SFX/Story/Waves/Quest_S36_snapshot_p01_q04_beacon_ambience_E_01}}
 {id=WavePlayer_15, properties {soundWaveAssetPtr=/BattlePassS36_Quests/Audio/SFX/Story/Waves/Quest_S36_snapshot_p01_q04_beacon_ambience_E_02}}
 {id=WavePlayer_2, properties {soundWaveAssetPtr=/BattlePassS36_Quests/Audio/SFX/Story/Waves/Quest_S36_snapshot_p01_q04_beacon_ambience_A_02}}
 {id=WavePlayer_3, properties {soundWaveAssetPtr=/BattlePassS36_Quests/Audio/SFX/Story/Waves/Quest_S36_snapshot_p01_q04_beacon_ambience_A_03}}
 {id=WavePlayer_4, properties {soundWaveAssetPtr=/BattlePassS36_Quests/Audio/SFX/Story/Waves/Quest_S36_snapshot_p01_q04_beacon_ambience_B_01}}
 {id=WavePlayer_5, properties {soundWaveAssetPtr=/BattlePassS36_Quests/Audio/SFX/Story/Waves/Quest_S36_snapshot_p01_q04_beacon_ambience_B_02}}
 {id=WavePlayer_6, properties {soundWaveAssetPtr=/BattlePassS36_Quests/Audio/SFX/Story/Waves/Quest_S36_snapshot_p01_q04_beacon_ambience_B_03}}
 {id=WavePlayer_7, properties {soundWaveAssetPtr=/BattlePassS36_Quests/Audio/SFX/Story/Waves/Quest_S36_snapshot_p01_q04_beacon_ambience_C_01}}
 {id=WavePlayer_8, properties {soundWaveAssetPtr=/BattlePassS36_Quests/Audio/SFX/Story/Waves/Quest_S36_snapshot_p01_q04_beacon_ambience_C_02}}
 {id=WavePlayer_9, properties {soundWaveAssetPtr=/BattlePassS36_Quests/Audio/SFX/Story/Waves/Quest_S36_snapshot_p01_q04_beacon_ambience_C_03}}
]
properties {
 attenuationSettings=/Game/Athena/Sounds/SupplyDrops/Fort_HiddenStar_Ambient_Loop_Att
 maxDistance=2500
 soundClassObject=&Fort_Gadgets
}
